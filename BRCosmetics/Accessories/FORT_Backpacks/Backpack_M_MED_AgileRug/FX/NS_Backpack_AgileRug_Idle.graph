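# Niagara system for the idle effect on the AgileRug backpack cosmetic, showing a persistent sprite and mesh visual tinted per variant.
# A single Minimal emitter spawns by instantaneous burst, and its particles live until the emitter deactivates, drawn by both a sprite and a mesh renderer.
# User.VariantID, a float cast to int, indexes a four-color array (two blues, red-orange, green) to set particle color at spawn.
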NiagaraSystem /BRCosmetics/Accessories/FORT_Backpacks/Backpack_M_MED_AgileRug/FX/NS_Backpack_AgileRug_Idle
dependencies {usedDataInterfaces [&NiagaraDataInterfaceArrayColor], usedDynamicInputs [], usedModules [&SystemState]}
emitters {Minimal {
  emitterUpdateScript [
   &EmitterState(Life Cycle Mode=::NewEnumerator1, Loop Behavior=::NewEnumerator1, Loop Duration Mode=::NewEnumerator1, UseLoopDelay=false)
   &SpawnBurst_Instantaneous(Use Spawn Probability=false, UseLoopCountLimit=true)
  ]
  particleSpawnScript [
   &InitializeParticle(
    Color=&SelectLinearColorFromArray(
     Array Sampling Mode=::NewEnumerator1
     Direct Array Index=&MakeIntFromFloat(Float=$User.VariantID)
     Color Selection Array=DataInterface() {propertyValues {ColorData [<rgba 0.2663 0.5273 1 0.3>, <rgba 0.2659 0.7049 1 0.3>, <rgba 1 0.2614 0.1431 0.85>, <rgba 0.5229 1 0.547 0.3>]}}
     Random Seed=?
    )
    Mesh Scale=<0.15 0.2 0.2>
    Mesh Scale Mode=::NewEnumerator4
    Sprite Size Mode=::NewEnumerator3
    Uniform Sprite Size=9
    UsePositionOffset=false
    Write Material Random=true
   )
   &InitialMeshOrientation(Mesh Axis To Orient=<0 1 0>, Mesh Orientation Mode=::NewEnumerator2, Use Rotation Vector=false)
  ]
  particleUpdateScript [&ParticleState(Kill Particles When Lifetime Has Elapsed=false, Let Infinitely Lived Particles Die When Emitter Deactivates=true)]
  rendererClasses [&NiagaraSpriteRendererProperties, &NiagaraMeshRendererProperties]
  renderers [
   {rendererClass=&NiagaraSpriteRendererProperties}
   {rendererClass=&NiagaraMeshRendererProperties, rendererIndex=1}
  ]
 }}
summary {
 systemName=NS_Backpack_AgileRug_Idle
 emitters [{emitterName=Minimal, rendererClasses [&NiagaraSpriteRendererProperties, &NiagaraMeshRendererProperties]}]
 userVariables [User.VariantID {type=NiagaraFloat, defaultValue=0, description=""}]
}
