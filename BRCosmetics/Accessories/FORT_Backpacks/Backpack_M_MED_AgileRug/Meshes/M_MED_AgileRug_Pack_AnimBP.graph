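# Anim Blueprint for the AgileRug backpack that offsets its pack bone in the pose.
# The AnimGraph passes the input pose through three chained Modify Bone nodes, each adding a bone-space Y translation to the pack bone.
# Each update, the EventGraph wildcard-matches the owning FortPlayerPawn's part mesh name to set is_AgileRug (excluding the Tuft variant) and FMedCharacters (F_MED names).
# The file shows no link from these flags to the bone nodes.
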AnimBlueprint /BRCosmetics/Accessories/FORT_Backpacks/Backpack_M_MED_AgileRug/Meshes/M_MED_AgileRug_Pack_AnimBP
target_skeleton=&/M_MED_AgileRug_Pack_Skeleton
graphs [
 AnimGraph {nodes [
   n1 = "Transform (Modify) Bone - Bone: pack"!(ComponentPose=Local To Component(LocalPose=Input Pose()), bAlphaBoolEnabled:bool, Translation="0,0.6,0", Rotation="0,0,0", Scale) -> (Pose) {
    alphaInputType=Bool
    translationMode=BMM_Additive
    translationSpace=BCS_BoneSpace
    alphaBoolBlend {blendOption=Linear}
    boneToModify {boneName=pack}
    translation {y=0.6}
   }
   n2 = "Transform (Modify) Bone - Bone: pack"(ComponentPose=@n1, Translation="0,1.3,0", Rotation="0,0,0") {
    alphaInputType=Bool
    translationMode=BMM_Additive
    translationSpace=BCS_BoneSpace
    alphaBoolBlend {blendOption=Linear}
    boneToModify {boneName=pack}
    translation {y=1.3}
   }
   n3 = "Transform (Modify) Bone - Bone: pack"(ComponentPose=@n2, Translation="0,1.3,0", Rotation="0,0,0") {
    alphaInputType=Bool
    translationMode=BMM_Additive
    translationSpace=BCS_BoneSpace
    alphaBoolBlend {blendOption=Linear}
    boneToModify {boneName=pack}
    translation {y=1.3}
   }
   Output Pose(Result=Component To Local(ComponentPose=@n3)) {layerGroup=None}
  ]}
 EventGraph {nodes [
   Event BlueprintUpdateAnimation()
   > n1 = Cast To FortPlayerPawn(Object=TryGetPawnOwner())
   n2 = GetDisplayName(Object=GetSkeletalMeshForPartType(self=@n1))
   n3 = MatchesWildcard(SourceString=@n2, Wildcard=*M_MED_AgileRug_Tuft*, SearchCase=CaseSensitive)
   > Set is_AgileRug(is_AgileRug:bool=AND Boolean(A=MatchesWildcard(SourceString=@n2, Wildcard=*M_MED_AgileRug*, SearchCase=CaseSensitive), B=NOT Boolean(A=@n3))) -> (Output_Get:bool)
   > Set FMedCharacters(FMedCharacters:bool=MatchesWildcard(SourceString=@n2, Wildcard=*F_MED*, SearchCase=CaseSensitive)) -> (Output_Get:bool)
  ]}
]
variables [
 FMedCharacters {type=boolean}
 is_AgileRug {type=boolean}
]
